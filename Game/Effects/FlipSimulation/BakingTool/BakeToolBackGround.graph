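# Unlit, two-sided material that serves, by its name, as the background for the FLIP simulation baking tool.
# Its graph holds a single Constant node left at its default value, with no textures or parameters.
# Any connection from that node to a material output is not shown.
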Material /Game/Effects/FlipSimulation/BakingTool/BakeToolBackGround
graph {properties {pixelDepthOffsetMode=PDOM_Legacy, shadingModel=MSM_Unlit, twoSided=true, usedShadingModels=Unlit}, nodes [Constant()]}
