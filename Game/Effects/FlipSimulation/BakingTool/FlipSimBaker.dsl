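; Flip simulation baker Blueprint that stages a Niagara system in front of an orthographic SceneCapture2D to render it into render targets.
; Initial assigns the system to the BakeTest component, sets ortho width from GridSize X, and offsets the component by the GridPivot's distance from center, scaled by GridSize.
; Reset restarts the system, and Capture renders one snapshot into the supplied target if valid.
; No loop or timing appears here.
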(fn Capture
  (params (NewParam :type Texture%20Render%20Target%202D%20Object%20Reference))
  (Utilities|IsValid NewParam
    (:Is%20Valid
      (Variables|/Script/Engine.SceneCaptureComponent2D|SetTextureTarget NewParam (Variables|Default|GetSceneCaptureComponent2D))
      (Rendering|SceneCapture|CaptureScene (Variables|Default|GetSceneCaptureComponent2D) :!class /Script/Engine.SceneCaptureComponent2D))))

(fn Initial
  (params (TargetNS :type Niagara%20System%20Object%20Reference))
  (bind (_x _y) (Math|Vector2D|BreakVector2D (* (- (Variables|Default|GetGridPivot) "(X=0.500000,Y=0.500000)") (Variables|Default|GetGridSize))))
  (bind _gridsize (Variables|Default|GetGridSize))
  (bind _baketest (Variables|Default|GetBakeTest))
  (Niagara|SetNiagaraSystemAsset (Variables|Default|GetBakeTest) TargetNS true)
  (Variables|/Script/Engine.SceneCaptureComponent2D|SetOrthoWidth (Math|Vector2D|BreakVector2D _gridsize) (Variables|Default|GetSceneCaptureComponent2D))
  (Transformation|SetRelativeLocation _baketest "800.000000,0.000000,0.000000" false false)
  (Transformation|AddLocalOffset _baketest (Math|Vector|MakeVector _x 0.0 _y) false false))

(fn Reset
  (Niagara|ResetSystem (Variables|Default|GetBakeTest)))

(fn ConstructionScript)
