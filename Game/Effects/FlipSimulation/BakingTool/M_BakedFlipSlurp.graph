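# Translucent unlit Niagara sprite material that plays baked FLIP fluid simulation flipbooks as refractive liquid.
# Particle Progress picks the frame, falling back to looping time, and a motion-vector texture warps it between frames without blending two frames.
# Thickness drives color and opacity, while baked normals feed fake lighting plus cubemap and skylight reflections.
# Dissolve noise, alpha contrast, edge masks and depth fade shape opacity and refraction.
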Material /Game/Effects/FlipSimulation/BakingTool/M_BakedFlipSlurp
graph {
 properties {
  bUsedWithNiagaraSprites=true
  blendMode=BLEND_Translucent
  pixelDepthOffsetMode=PDOM_Legacy
  refractionMethod=RM_IndexOfRefraction
  shadingModel=MSM_Unlit
  twoSided=true
  usedShadingModels=Unlit
 }
 nodes [
  n1 = TextureObjectParameter() {
   group=Dissolve
   parameterName=Dissolve_Texture
   samplerType=SAMPLERTYPE_LinearColor
   sortPriority=0
   texture=/Game/Effects/Fort_Effects/Textures/T_FX_IndieBucket_Contrail_Cells
  }
  n2 = VectorParameter() {group=Dissolve, parameterName=Dissolve_UVControl, channelNames {a=VPan, b=UPan, g=VScale, r=UScale}, defaultValue=<rgba 1 1 0 0>}
  n3 = MaterialFunctionCall(Texture=@n1, UVControl_RGB=@n2, UVControl_A=@n2) {materialFunction=&FXCommon_TextureSample}
  n4 = DynamicParameter() {defaultValue=<rgba 0 0 0 1>, paramNames [Progress,Dissolve,EdgeOffset,Param4]}
  n5 = ScalarParameter() {defaultValue=-1, parameterName=DissolveTest}
  n6 = MaterialFunctionCall(
   In=Subtract(A=OneMinus(None=@n3.R), B=Add(A=@n4.Dissolve, B=@n5))
   Contrast=ScalarParameter() {defaultValue=5, parameterName=DissolveContrast}
  ) {materialFunction=&CheapContrast}
  n7 = VectorParameter() {parameterName=SubUVSize, defaultValue=<rgba 4 4 0 1>}
  n8 = Frac(None=Time() {bOverride_Period=true, period=666})
  n9 = LinearInterpolate(B=Subtract(A=Multiply(A=@n7.R, B=@n7.R)), Alpha=If(A=@n4.Progress, "A > B"=@n4.Progress, "A == B"=@n8, "A < B"=@n8))
  n10 = Frac(None=@n9) {desc=Frame Lerp}
  n11 = Subtract(A=@n9, B=@n10) {desc=IndexA}
  n12 = AppendVector(A=@n7.R, B=@n7.R)
  n13 = Fmod(A=@n11, B=ComponentMask(None=@n12) {desc="## Horz Images", r=true}) {desc=IndexA_Horz}
  n14 = Divide(A=Constant2Vector() {g=1, r=1}, B=@n12) {desc=Scale}
  n15 = Floor(None=Multiply(A=@n11, B=ComponentMask(None=@n14) {desc="1 / ## Horz Images", r=true})) {desc=IndexA_Vert}
  n16 = Multiply(A=Add(A=AppendVector(A=@n13, B=@n15) {desc=IndexA_Combined}, B=TextureCoordinate()), B=@n14) {desc=UV_A}
  n17 = TextureSampleParameter2D(UVs=@n16) {parameterName=MotionVectorTexture, samplerType=SAMPLERTYPE_Normal, texture=&FlatNormalBlue}
  n18 = AppendVector(A=@n17.R, B=@n17.R)
  n19 = Add(A=@n16, B=Multiply(A=Multiply(A=@n18, B=@n10), B=ScalarParameter() {defaultValue=-0.03, parameterName=MotionVectorStrength}))
  n20 = TextureObjectParameter() {parameterName=ThicknessAndAlphaTexture, samplerType=SAMPLERTYPE_LinearColor, texture=&/T_BlankGray}
  n21 = TextureSample(UVs=@n19, Tex=@n20) {samplerType=SAMPLERTYPE_LinearColor, texture=/Game/Athena/UI/InGame/Sidekick/Reactions/100}
  n22 = Multiply(A=@n6.Result, B=@n21.R)
  n23 = Power(
   Base=DepthFade(FadeDistance=ScalarParameter() {defaultValue=150, parameterName=FadeDistance})
   Exp=ScalarParameter() {defaultValue=2, parameterName=DepthFadePower}
  )
  n24 = MaterialFunctionCall(Float2=TextureCoordinate()) {materialFunction=&BreakOutFloat2Components}
  n25 = VectorParameter() {group=EdgeFade, parameterName=EdgeFade_EdgeMask_Control, channelNames {a="", b=Power, g=Multiply, r=Offset}, defaultValue=<rgba 0 30 4 1>}
  n26 = Multiply(
   A=MaterialFunctionCall!(UorV=@n24.R, EdgeControl=@n25) {materialFunction=/Game/Effects/Niagara/MasterMaterials/Functions/Common/FXCommon_EdgeMask_Simple}.Mask
   B=MaterialFunctionCall(UorV=@n24.G, EdgeControl=@n25) {materialFunction=&FXCommon_EdgeMask_Simple}.Mask
  )
  n27 = ParticleColor()
  n28 = ScalarParameter() {parameterName=UseParticleAlpha}
  LinearInterpolate(
   B=LinearInterpolate(B=ScalarParameter() {defaultValue=1.1, parameterName=Refraction}, Alpha=Multiply(A=@n22, B=@n23)) {constA=1}
   Alpha=Multiply(A=@n26, B=LinearInterpolate(B=@n27.A, Alpha=@n28) {constA=1})
  ) {constA=1}
  n29 = TextureSampleParameter2D(UVs=@n19) {parameterName=NormalTexture, samplerType=SAMPLERTYPE_Normal, texture=&FlatNormalBlue}
  Multiply(A=Subtract(A=@n29.RGB) {constB=0.5}) {constB=2}
  n30 = ScalarParameter() {defaultValue=0.1, parameterName=ReflectionHDRIStr}
  n31 = ScalarParameter() {parameterName=Reflection_NormalFlatten}
  n32 = ReflectionVectorWS(CustomWorldNormal=Transform(None=MaterialFunctionCall(Normal=@n29.RGB, Flatness=@n31) {materialFunction=&FlattenNormal}.Result))
  n33 = TextureSampleParameterCube!(UVs=@n32, Apply View MipBias) {parameterName=ReflectHDRI, samplerType=SAMPLERTYPE_LinearColor, texture=/Game/Effects/FlipSimulation/Rendering/T_WaterReflection}
  n34 = VectorParameter() {parameterName=ColorB, defaultValue=<rgba 0 1 0.6667 4>}
  n35 = VectorParameter() {parameterName=ColorA, defaultValue=<rgba 0 0.6667 0.8 4>}
  n36 = Multiply(
   A=Add(A=@n21.R, B=ScalarParameter() {group=ThicknessToColor, parameterName=ThicknessToColor_GrayAdjust_Add})
   B=ScalarParameter() {defaultValue=1.8, group=ThicknessToColor, parameterName=ThicknessToColor_GrayAdjust_Mul}
  )
  n37 = Saturate(None=Power(Base=Saturate(None=@n36), Exp=ScalarParameter() {defaultValue=1.2, group=ThicknessToColor, parameterName=ThicknessToColor_GrayAdjust_Power}))
  n38 = Normalize(VectorInput=VectorParameter() {parameterName=FakeLightDirection, defaultValue=<rgba 0 0.5 1 1>}.RGB)
  n39 = Divide(A=LinearInterpolate(A=@n34.RGB, B=LinearInterpolate(A=@n35.RGB, B=@n34.RGB, Alpha=@n37), Alpha=DotProduct(A=@n38, B=@n29.RGB)))
  n40 = ParticleColor()
  n41 = MaterialFunctionCall(Emissive=LinearInterpolate(A=@n39, B=Multiply(A=@n39, B=@n40.RGB), Alpha=ScalarParameter() {parameterName=UseParticleColor})) {materialFunction=&MF_FX_EmissiveAdjustments}
  n42 = Multiply(A=ScalarParameter() {defaultValue=0.1, parameterName=ReflectionSkySTr}, B=SkyLightEnvMapSample(Direction=@n32)) {constA=0.2}
  Reroute(None=Add(A=Multiply(A=@n30, B=@n33.RGB), B=Add(A=@n41.Result, B=@n42)))
  n43 = ScalarParameter() {defaultValue=1.7, parameterName=ThicknessToAlpha_Mul}
  n44 = Saturate(None=Add(A=Multiply(A=@n43, B=@n21.R), B=ScalarParameter() {defaultValue=0.5, parameterName=ThicknessToAlpha_Base}))
  n45 = ScalarParameter() {defaultValue=2.2, parameterName=ThicknessToAlpha_Pow}
  n46 = Add(
   A=@n21.A
   B=Add(
    A=DynamicParameter() {defaultValue=<rgba 0 0 0 1>, paramNames [Progress,Dissolve,EdgeOffset,Param4]}.EdgeOffset
    B=ScalarParameter() {parameterName=AlphaOffset}
   )
  )
  n47 = ScalarParameter() {defaultValue=2, parameterName=AlphaContrast}
  n48 = Saturate(None=Multiply(
    A=Power(Base=@n44, Exp=@n45)
    B=MaterialFunctionCall(In=@n46, Contrast=@n47) {materialFunction=&CheapContrast}.Result
   ))
  n49 = Constant() {r=1}
  n50 = Multiply(
   A=LinearInterpolate(B=@n40.A, Alpha=ScalarParameter() {parameterName=UseParticleAlpha}) {constA=1}
   B=Multiply(A=Multiply(A=@n6.Result, B=@n48), B=StaticSwitchParameter(True=@n49, False=@n23) {parameterName=UseMaskedBlendMode})
  )
  n51 = Constant() {r=1}
  StaticSwitchParameter(
   True=Multiply(
    A=Multiply(A=@n50, B=@n26)
    B=MaterialFunctionCall("FadeSensitivity (optional)"=@n51) {materialFunction=&FadeBasedOnViewAngle}.Result
   )
   False=Constant() {r=1}
  ) {parameterName=UseMaskedBlendMode}
 ]
}
