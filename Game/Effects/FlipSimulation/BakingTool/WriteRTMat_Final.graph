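# WriteRTMat_Final is an unlit, translucent, two-sided material in the flip simulation baking tool, apparently for writing render target content.
# Its graph samples a texture object parameter named RT at the default texture coordinates, using the linear color sampler type.
# The graph shows no connection from that sample to a material output and no other processing.
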Material /Game/Effects/FlipSimulation/BakingTool/WriteRTMat_Final
graph {
 properties {blendMode=BLEND_Translucent, pixelDepthOffsetMode=PDOM_Legacy, shadingModel=MSM_Unlit, twoSided=true, usedShadingModels=Unlit}
 nodes [TextureSample(UVs=TextureCoordinate(), Tex=TextureObjectParameter() {parameterName=RT}) {samplerType=SAMPLERTYPE_LinearColor}]
}
