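# Translucent post-process material in the FlipSimulation baking tool that reads a render target texture parameter named RT as a 2x2 grid of quadrants.
# Halved UVs plus offsets sample the red channel of three quadrants.
# The bottom-left value is divided by the top-right, and the bottom-right is duplicated into two channels.
# No material output connection appears, so these branches look unfinished or unused.
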Material /Game/Effects/FlipSimulation/BakingTool/WriteRTMat_T
graph {
 properties {
  blendMode=BLEND_Translucent
  blendableOutputAlpha=true
  materialDomain=MD_PostProcess
  pixelDepthOffsetMode=PDOM_Legacy
  shadingModel=MSM_Unlit
  twoSided=true
  usedShadingModels=Unlit
 }
 nodes [
  n1 = Multiply(A=TextureCoordinate()) {constB=0.5}
  n2 = TextureObjectParameter() {parameterName=RT}
  n3 = TextureSample(UVs=Add(A=@n1, B=Constant2Vector() {g=0.5, r=0.5}), Tex=@n2) {samplerType=SAMPLERTYPE_LinearColor}
  AppendVector(A=@n3.R, B=@n3.R)
  Constant() {r=1}
  n4 = Constant2Vector() {g=0.5}
  n5 = Constant2Vector() {r=0.5}
  Divide(
   A=TextureSample(UVs=Add(A=@n1, B=@n4), Tex=@n2) {samplerType=SAMPLERTYPE_LinearColor}.R
   B=TextureSample(UVs=Add(A=@n1, B=@n5), Tex=@n2) {samplerSource=SSM_Clamp_WorldGroupSettings, samplerType=SAMPLERTYPE_LinearColor}.R
  )
 ]
}
